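# Third-person sound cue for an enemy soldier's hard landing on stone, using "Downstairs" samples for the below-elevation variant.
# Each mix layers a random stone landing and a louder random thump, both modulated for pitch and volume.
# A quality switch gives the first two levels six variations per layer and the last two a reduced pool of three landings and two thumps.
# Landing attenuation and owner-limited, stop-oldest concurrency govern playback.
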SoundCue /Game/Athena/Sounds/Character/Footsteps/3P/Stone/Athena_Soldier_Stone_LandHard_Below_3P_Cue
root=QualityLevel_0
nodes [
 {id=Mixer_0, children [Modulator_2,Modulator_1], properties {inputVolume [1,1.6]}}
 {id=Mixer_5, children [Modulator_0,Modulator_5], properties {inputVolume [1,1.6]}}
 {id=Modulator_0, children [Random_4], properties {pitchMax=1.1, pitchMin=0.9, volumeMax=1.3, volumeMin=0.7}}
 {id=Modulator_1, children [Random_0], properties {pitchMax=1.1, pitchMin=0.9, volumeMax=1.3, volumeMin=0.7}}
 {id=Modulator_2, children [Random_1], properties {pitchMax=1.1, pitchMin=0.9, volumeMax=1.3, volumeMin=0.7}}
 {id=Modulator_5, children [Random_2], properties {pitchMax=1.1, pitchMin=0.9, volumeMax=1.3, volumeMin=0.7}}
 {id=QualityLevel_0, children [Mixer_5,Mixer_5,Mixer_0,Mixer_0]}
 {id=Random_0, children [WavePlayer_1,WavePlayer_6,WavePlayer_7], properties {bShouldExcludeFromBranchCulling=true, weights [1,1,1]}}
 {id=Random_1, children [WavePlayer_9, WavePlayer_10], properties {bShouldExcludeFromBranchCulling=true, weights [1,1]}}
 {
  id=Random_2
  children [WavePlayer_3, WavePlayer_4, WavePlayer_5, WavePlayer_14, WavePlayer_15, WavePlayer_16]
  properties {weights [1,1,1,1,1,1]}
 }
 {
  id=Random_4
  children [WavePlayer_23, WavePlayer_24, WavePlayer_25, WavePlayer_26, WavePlayer_27, WavePlayer_28]
  properties {weights [1,1,1,1,1,1]}
 }
 {
  id=WavePlayer_1
  properties {soundWaveAssetPtr=/Game/Sounds/Fort_Step_Sounds/Soldier/Elevations/Downstairs/Soldier_Step_Land_Stone_01_Downstairs}
 }
 {id=WavePlayer_10, properties {soundWaveAssetPtr=/Game/Sounds/Fort_Step_Sounds/Soldier/Elevations/Downstairs/FS_Downstairs_Stone_Thump_02}}
 {
  id=WavePlayer_14
  properties {soundWaveAssetPtr=/Game/Sounds/Fort_Step_Sounds/Soldier/Elevations/Downstairs/Soldier_Step_Land_Stone_02_Downstairs}
 }
 {
  id=WavePlayer_15
  properties {soundWaveAssetPtr=/Game/Sounds/Fort_Step_Sounds/Soldier/Elevations/Downstairs/Soldier_Step_Land_Stone_03_Downstairs}
 }
 {
  id=WavePlayer_16
  properties {soundWaveAssetPtr=/Game/Sounds/Fort_Step_Sounds/Soldier/Elevations/Downstairs/Soldier_Step_Land_Stone_05_Downstairs}
 }
 {id=WavePlayer_23, properties {soundWaveAssetPtr=/Game/Sounds/Fort_Step_Sounds/Soldier/Elevations/Downstairs/FS_Downstairs_Stone_Thump_01}}
 {id=WavePlayer_24, properties {soundWaveAssetPtr=&FS_Downstairs_Stone_Thump_02}}
 {id=WavePlayer_25, properties {soundWaveAssetPtr=/Game/Sounds/Fort_Step_Sounds/Soldier/Elevations/Downstairs/FS_Downstairs_Stone_Thump_03}}
 {id=WavePlayer_26, properties {soundWaveAssetPtr=/Game/Sounds/Fort_Step_Sounds/Soldier/Elevations/Downstairs/FS_Downstairs_Stone_Thump_04}}
 {id=WavePlayer_27, properties {soundWaveAssetPtr=/Game/Sounds/Fort_Step_Sounds/Soldier/Elevations/Downstairs/FS_Downstairs_Stone_Thump_05}}
 {id=WavePlayer_28, properties {soundWaveAssetPtr=/Game/Sounds/Fort_Step_Sounds/Soldier/Elevations/Downstairs/FS_Downstairs_Stone_Thump_06}}
 {
  id=WavePlayer_3
  properties {soundWaveAssetPtr=/Game/Sounds/Fort_Step_Sounds/Soldier/Elevations/Downstairs/Soldier_Step_Land_Stone_06_Downstairs}
 }
 {
  id=WavePlayer_4
  properties {soundWaveAssetPtr=&Soldier_Step_Land_Stone_01_Downstairs}
 }
 {
  id=WavePlayer_5
  properties {soundWaveAssetPtr=/Game/Sounds/Fort_Step_Sounds/Soldier/Elevations/Downstairs/Soldier_Step_Land_Stone_04_Downstairs}
 }
 {
  id=WavePlayer_6
  properties {soundWaveAssetPtr=&Soldier_Step_Land_Stone_02_Downstairs}
 }
 {
  id=WavePlayer_7
  properties {soundWaveAssetPtr=&Soldier_Step_Land_Stone_03_Downstairs}
 }
 {id=WavePlayer_9, properties {soundWaveAssetPtr=&FS_Downstairs_Stone_Thump_01}}
]
properties {
 attenuationSettings=/Game/Sounds/Attenuation/Fort_Player_Land_3P_Athena_Att
 maxDistance=3584
 soundClassObject=&Fort_Player_Foley_3P_Enemy_Athena
 volumeMultiplier=1.2
 concurrencyOverrides {bLimitToOwner=true, maxCount=4, resolutionRule=StopOldest}
 concurrencySet [&Athena_Footstep_Global_Conc]
}
