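# Sound cue for third-person soldier sprint footsteps on stone heard from above in Athena, built from "Upstairs" stone walk samples.
# A quality-level switch picks an equally weighted random pool: twelve waves for the first two levels, three for the last two.
# A modulator applies a fixed volume boost.
# Sprint footstep attenuation, an enemy foley sound class, and local and global footstep concurrency sets govern playback.
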SoundCue /Game/Athena/Sounds/Character/Footsteps/3P/Stone/Athena_Soldier_Stone_Sprint_Footstep_Above_3P_Cue
root=Modulator_3
nodes [
 {id=Modulator_3, children [QualityLevel_1], properties {volumeMax=1.2, volumeMin=1.2}}
 {id=QualityLevel_1, children [Random_1,Random_1,Random_0,Random_0]}
 {id=Random_0, children [WavePlayer_9, WavePlayer_11, WavePlayer_12], properties {bShouldExcludeFromBranchCulling=true, weights [1,1,1]}}
 {
  id=Random_1
  children [
   WavePlayer_2
   WavePlayer_3
   WavePlayer_4
   WavePlayer_5
   WavePlayer_7
   WavePlayer_10
   WavePlayer_18
   WavePlayer_19
   WavePlayer_0
   WavePlayer_1
   WavePlayer_6
   WavePlayer_8
  ]
  properties {weights [1,1,1,1,1,1,1,1,1,1,1,1]}
 }
 {
  id=WavePlayer_0
  properties {soundWaveAssetPtr=/Game/Sounds/Fort_Step_Sounds/Soldier/Elevations/Upstairs/Soldier_Step_Walk_Stone_09_Upstairs2}
 }
 {
  id=WavePlayer_1
  properties {soundWaveAssetPtr=/Game/Sounds/Fort_Step_Sounds/Soldier/Elevations/Upstairs/Soldier_Step_Walk_Stone_10_Upstairs2}
 }
 {
  id=WavePlayer_10
  properties {soundWaveAssetPtr=/Game/Sounds/Fort_Step_Sounds/Soldier/Elevations/Upstairs/Soldier_Step_Walk_Stone_06_Upstairs2}
 }
 {
  id=WavePlayer_11
  properties {soundWaveAssetPtr=/Game/Sounds/Fort_Step_Sounds/Soldier/Elevations/Upstairs/Soldier_Step_Walk_Stone_02_Upstairs2}
 }
 {
  id=WavePlayer_12
  properties {soundWaveAssetPtr=/Game/Sounds/Fort_Step_Sounds/Soldier/Elevations/Upstairs/Soldier_Step_Walk_Stone_03_Upstairs2}
 }
 {
  id=WavePlayer_18
  properties {soundWaveAssetPtr=/Game/Sounds/Fort_Step_Sounds/Soldier/Elevations/Upstairs/Soldier_Step_Walk_Stone_07_Upstairs2}
 }
 {
  id=WavePlayer_19
  properties {soundWaveAssetPtr=/Game/Sounds/Fort_Step_Sounds/Soldier/Elevations/Upstairs/Soldier_Step_Walk_Stone_08_Upstairs2}
 }
 {
  id=WavePlayer_2
  properties {soundWaveAssetPtr=/Game/Sounds/Fort_Step_Sounds/Soldier/Elevations/Upstairs/Soldier_Step_Walk_Stone_01_Upstairs2}
 }
 {
  id=WavePlayer_3
  properties {soundWaveAssetPtr=&Soldier_Step_Walk_Stone_02_Upstairs2}
 }
 {
  id=WavePlayer_4
  properties {soundWaveAssetPtr=&Soldier_Step_Walk_Stone_03_Upstairs2}
 }
 {
  id=WavePlayer_5
  properties {soundWaveAssetPtr=/Game/Sounds/Fort_Step_Sounds/Soldier/Elevations/Upstairs/Soldier_Step_Walk_Stone_04_Upstairs2}
 }
 {
  id=WavePlayer_6
  properties {soundWaveAssetPtr=/Game/Sounds/Fort_Step_Sounds/Soldier/Elevations/Upstairs/Soldier_Step_Walk_Stone_11_Upstairs2}
 }
 {
  id=WavePlayer_7
  properties {soundWaveAssetPtr=/Game/Sounds/Fort_Step_Sounds/Soldier/Elevations/Upstairs/Soldier_Step_Walk_Stone_05_Upstairs2}
 }
 {
  id=WavePlayer_8
  properties {soundWaveAssetPtr=/Game/Sounds/Fort_Step_Sounds/Soldier/Elevations/Upstairs/Soldier_Step_Walk_Stone_12_Upstairs2}
 }
 {
  id=WavePlayer_9
  properties {soundWaveAssetPtr=&Soldier_Step_Walk_Stone_01_Upstairs2}
 }
]
properties {
 attenuationSettings=/Game/Sounds/Attenuation/Fort_Player_Footsteps_3P_Sprint_Athena_Att
 maxDistance=3584
 soundClassObject=&Fort_Player_Foley_3P_Enemy_Athena
 volumeMultiplier=0.9
 concurrencyOverrides {bLimitToOwner=true, maxCount=1, resolutionRule=StopOldest}
 concurrencySet [&Athena_Footstep_Local_Conc, &Athena_Footstep_Global_Conc]
}
