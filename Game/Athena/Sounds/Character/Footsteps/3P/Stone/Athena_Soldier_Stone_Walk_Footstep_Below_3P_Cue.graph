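# Sound cue for third-person soldier walking footsteps on stone heard from below, using downstairs wave variants and the enemy foley sound class.
# A quality-level node sends its first two levels to a random pick of twelve waves and its last two to just three.
# Each random pick passes through a modulator for variation.
# Walk-specific footstep attenuation and local plus global footstep concurrency sets control range and voice count.
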SoundCue /Game/Athena/Sounds/Character/Footsteps/3P/Stone/Athena_Soldier_Stone_Walk_Footstep_Below_3P_Cue
root=QualityLevel_0
nodes [
 {id=Modulator_0, children [Random_0]}
 {id=Modulator_1, children [Random_1]}
 {id=QualityLevel_0, children [Modulator_1,Modulator_1,Modulator_0,Modulator_0]}
 {id=Random_0, children [WavePlayer_0,WavePlayer_1,WavePlayer_2], properties {bShouldExcludeFromBranchCulling=true, weights [1,1,1]}}
 {
  id=Random_1
  children [
   WavePlayer_30
   WavePlayer_31
   WavePlayer_32
   WavePlayer_33
   WavePlayer_34
   WavePlayer_35
   WavePlayer_37
   WavePlayer_38
   WavePlayer_39
   WavePlayer_40
   WavePlayer_41
   WavePlayer_42
  ]
  properties {weights [1,1,1,1,1,1,1,1,1,1,1,1]}
 }
 {
  id=WavePlayer_0
  properties {soundWaveAssetPtr=/Game/Sounds/Fort_Step_Sounds/Soldier/Elevations/Downstairs/Soldier_Step_Walk_Stone_01_Downstairs}
 }
 {
  id=WavePlayer_1
  properties {soundWaveAssetPtr=/Game/Sounds/Fort_Step_Sounds/Soldier/Elevations/Downstairs/Soldier_Step_Walk_Stone_02_Downstairs}
 }
 {
  id=WavePlayer_2
  properties {soundWaveAssetPtr=/Game/Sounds/Fort_Step_Sounds/Soldier/Elevations/Downstairs/Soldier_Step_Walk_Stone_03_Downstairs}
 }
 {
  id=WavePlayer_30
  properties {soundWaveAssetPtr=&Soldier_Step_Walk_Stone_01_Downstairs}
 }
 {
  id=WavePlayer_31
  properties {soundWaveAssetPtr=&Soldier_Step_Walk_Stone_02_Downstairs}
 }
 {
  id=WavePlayer_32
  properties {soundWaveAssetPtr=&Soldier_Step_Walk_Stone_03_Downstairs}
 }
 {
  id=WavePlayer_33
  properties {soundWaveAssetPtr=/Game/Sounds/Fort_Step_Sounds/Soldier/Elevations/Downstairs/Soldier_Step_Walk_Stone_04_Downstairs}
 }
 {
  id=WavePlayer_34
  properties {soundWaveAssetPtr=/Game/Sounds/Fort_Step_Sounds/Soldier/Elevations/Downstairs/Soldier_Step_Walk_Stone_05_Downstairs}
 }
 {
  id=WavePlayer_35
  properties {soundWaveAssetPtr=/Game/Sounds/Fort_Step_Sounds/Soldier/Elevations/Downstairs/Soldier_Step_Walk_Stone_06_Downstairs}
 }
 {
  id=WavePlayer_37
  properties {soundWaveAssetPtr=/Game/Sounds/Fort_Step_Sounds/Soldier/Elevations/Downstairs/Soldier_Step_Walk_Stone_07_Downstairs}
 }
 {
  id=WavePlayer_38
  properties {soundWaveAssetPtr=/Game/Sounds/Fort_Step_Sounds/Soldier/Elevations/Downstairs/Soldier_Step_Walk_Stone_08_Downstairs}
 }
 {
  id=WavePlayer_39
  properties {soundWaveAssetPtr=/Game/Sounds/Fort_Step_Sounds/Soldier/Elevations/Downstairs/Soldier_Step_Walk_Stone_09_Downstairs}
 }
 {
  id=WavePlayer_40
  properties {soundWaveAssetPtr=/Game/Sounds/Fort_Step_Sounds/Soldier/Elevations/Downstairs/Soldier_Step_Walk_Stone_10_Downstairs}
 }
 {
  id=WavePlayer_41
  properties {soundWaveAssetPtr=/Game/Sounds/Fort_Step_Sounds/Soldier/Elevations/Downstairs/Soldier_Step_Walk_Stone_11_Downstairs}
 }
 {
  id=WavePlayer_42
  properties {soundWaveAssetPtr=/Game/Sounds/Fort_Step_Sounds/Soldier/Elevations/Downstairs/Soldier_Step_Walk_Stone_12_Downstairs}
 }
]
properties {
 attenuationSettings=/Game/Sounds/Attenuation/Fort_Player_Footsteps_3P_Walk_Athena_Att
 maxDistance=3584
 soundClassObject=&Fort_Player_Foley_3P_Enemy_Athena
 volumeMultiplier=0.85
 concurrencyOverrides {bLimitToOwner=true, maxCount=1, resolutionRule=StopOldest}
 concurrencySet [&Athena_Footstep_Local_Conc, &Athena_Footstep_Global_Conc]
}
