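# Third-person enemy landing sound cue for henchman footsteps on stone, built by reusing Walk and Run waves.
# Each mixer layers a random, pitched-up Walk sample with a slightly delayed random Run sample under an outer modulator.
# A quality switch gives its first two levels four variants per layer and the last two three, whose random nodes skip branch culling.
# Land attenuation and owner-limited, stop-oldest concurrency govern playback.
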SoundCue /Game/Athena/Sounds/Character/Footsteps/3P/Stone/FS_Henchman_Stone_Land_Cue
root=Modulator_2
nodes [
 {id=Delay_0, children [Random_3], properties {delayMax=0.1, delayMin=0.07}}
 {id=Delay_1, children [Random_2], properties {delayMax=0.1, delayMin=0.07}}
 {id=Mixer_0, children [Modulator_1,Delay_1], properties {inputVolume [1,1]}}
 {id=Mixer_1, children [Modulator_0,Delay_0], properties {inputVolume [1,1]}}
 {id=Modulator_0, children [Random_0], properties {pitchMax=1.2, pitchMin=1.1}}
 {id=Modulator_1, children [Random_1], properties {pitchMax=1.2, pitchMin=1.1}}
 {id=Modulator_2, children [QualityLevel_0]}
 {id=QualityLevel_0, children [Mixer_1,Mixer_1,Mixer_0,Mixer_0]}
 {id=Random_0, children [WavePlayer_0,WavePlayer_1,WavePlayer_2,WavePlayer_3], properties {weights [1,1,1,1]}}
 {id=Random_1, children [WavePlayer_8, WavePlayer_9, WavePlayer_10], properties {bShouldExcludeFromBranchCulling=true, weights [1,1,1]}}
 {id=Random_2, children [WavePlayer_12, WavePlayer_13, WavePlayer_14], properties {bShouldExcludeFromBranchCulling=true, weights [1,1,1]}}
 {id=Random_3, children [WavePlayer_4,WavePlayer_5,WavePlayer_6,WavePlayer_7], properties {weights [1,1,1,1]}}
 {id=WavePlayer_0, properties {soundWaveAssetPtr=&/FS_Henchman_Stone_Walk_01}}
 {id=WavePlayer_1, properties {soundWaveAssetPtr=&/FS_Henchman_Stone_Walk_04}}
 {id=WavePlayer_10, properties {soundWaveAssetPtr=&/FS_Henchman_Stone_Walk_03}}
 {id=WavePlayer_12, properties {soundWaveAssetPtr=&/FS_Henchman_Stone_Run_01}}
 {id=WavePlayer_13, properties {soundWaveAssetPtr=&/FS_Henchman_Stone_Run_04}}
 {id=WavePlayer_14, properties {soundWaveAssetPtr=&/FS_Henchman_Stone_Run_03}}
 {id=WavePlayer_2, properties {soundWaveAssetPtr=&FS_Henchman_Stone_Walk_03}}
 {id=WavePlayer_3, properties {soundWaveAssetPtr=&/FS_Henchman_Stone_Walk_02}}
 {id=WavePlayer_4, properties {soundWaveAssetPtr=&FS_Henchman_Stone_Run_01}}
 {id=WavePlayer_5, properties {soundWaveAssetPtr=&FS_Henchman_Stone_Run_04}}
 {id=WavePlayer_6, properties {soundWaveAssetPtr=&FS_Henchman_Stone_Run_03}}
 {id=WavePlayer_7, properties {soundWaveAssetPtr=&/FS_Henchman_Stone_Run_02}}
 {id=WavePlayer_8, properties {soundWaveAssetPtr=&FS_Henchman_Stone_Walk_01}}
 {id=WavePlayer_9, properties {soundWaveAssetPtr=&FS_Henchman_Stone_Walk_04}}
]
properties {
 attenuationSettings=/Game/Sounds/Attenuation/Fort_Player_Land_3P_Athena_Att
 bPrimeOnLoad=true
 maxDistance=3584
 soundClassObject=&Fort_Player_Foley_3P_Enemy_Athena
 volumeMultiplier=1
 concurrencyOverrides {bLimitToOwner=true, maxCount=4, resolutionRule=StopOldest}
 concurrencySet [&Athena_Footstep_Global_Conc]
}
