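# Sound cue for a henchman's third-person hard landing on stone, routed through the enemy foley class with player landing attenuation.
# Each variant mixes a quieter Commando landing foley with a louder, slightly delayed random stone run footstep.
# A quality-level switch under a modulator gives the first two levels five footstep variations and the last two a three-sample pool.
# Owner-limited concurrency stops the oldest instance, alongside a global footstep concurrency set.
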SoundCue /Game/Athena/Sounds/Character/Footsteps/3P/Stone/FS_Henchman_Stone_Land_Hard_Cue
root=Modulator_0
nodes [
 {id=Delay_0, children [Random_0], properties {delayMax=0.11, delayMin=0.08}}
 {id=Delay_1, children [Random_1], properties {delayMax=0.11, delayMin=0.08}}
 {id=Mixer_0, children [WavePlayer_0,Delay_1], properties {inputVolume [0.5,2]}}
 {id=Mixer_1, children [WavePlayer_5,Delay_0], properties {inputVolume [0.5,2]}}
 {id=Modulator_0, children [QualityLevel_0]}
 {id=QualityLevel_0, children [Mixer_1,Mixer_1,Mixer_0,Mixer_0]}
 {id=Random_0, children [WavePlayer_1,WavePlayer_2,WavePlayer_3,WavePlayer_4,WavePlayer_6], properties {weights [1,1,1,1,1]}}
 {id=Random_1, children [WavePlayer_7,WavePlayer_8,WavePlayer_9], properties {bShouldExcludeFromBranchCulling=true, weights [1,1,1]}}
 {id=WavePlayer_0, properties {soundWaveAssetPtr=/Game/Sounds/Fortress_Foley/Commando/Commando_Land_005}}
 {id=WavePlayer_1, properties {soundWaveAssetPtr=&/FS_Henchman_Stone_Run_09}}
 {id=WavePlayer_2, properties {soundWaveAssetPtr=&/FS_Henchman_Stone_Run_08}}
 {id=WavePlayer_3, properties {soundWaveAssetPtr=&/FS_Henchman_Stone_Run_07}}
 {id=WavePlayer_4, properties {soundWaveAssetPtr=&/FS_Henchman_Stone_Run_06}}
 {id=WavePlayer_5, properties {soundWaveAssetPtr=&Commando_Land_005}}
 {id=WavePlayer_6, properties {soundWaveAssetPtr=&/FS_Henchman_Stone_Run_05}}
 {id=WavePlayer_7, properties {soundWaveAssetPtr=&FS_Henchman_Stone_Run_09}}
 {id=WavePlayer_8, properties {soundWaveAssetPtr=&FS_Henchman_Stone_Run_08}}
 {id=WavePlayer_9, properties {soundWaveAssetPtr=&FS_Henchman_Stone_Run_07}}
]
properties {
 attenuationSettings=/Game/Sounds/Attenuation/Fort_Player_Land_3P_Athena_Att
 bPrimeOnLoad=true
 maxDistance=3584
 soundClassObject=&Fort_Player_Foley_3P_Enemy_Athena
 volumeMultiplier=1.4
 concurrencyOverrides {bLimitToOwner=true, maxCount=4, resolutionRule=StopOldest}
 concurrencySet [&Athena_Footstep_Global_Conc]
}
