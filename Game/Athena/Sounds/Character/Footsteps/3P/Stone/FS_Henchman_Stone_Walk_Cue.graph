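# Sound cue for third-person henchman walking footsteps on stone, routed to the enemy foley sound class.
# A quality-level node gives the first level a random pick of twelve waves, while the other three levels share a four-wave random node.
# A modulator varies pitch and volume on top.
# The cue primes on load, uses the third-person sprint footstep attenuation, and references local and global footstep concurrency sets.
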SoundCue /Game/Athena/Sounds/Character/Footsteps/3P/Stone/FS_Henchman_Stone_Walk_Cue
root=Modulator_2
nodes [
 {id=Modulator_2, children [QualityLevel_0], properties {pitchMax=0.9, pitchMin=1, volumeMax=1, volumeMin=0.7}}
 {id=QualityLevel_0, children [Random_1,Random_0,Random_0,Random_0]}
 {
  id=Random_0
  children [WavePlayer_5,WavePlayer_6,WavePlayer_7,WavePlayer_8]
  properties {bShouldExcludeFromBranchCulling=true, weights [1,1,1,1]}
 }
 {
  id=Random_1
  children [
   WavePlayer_9
   WavePlayer_10
   WavePlayer_11
   WavePlayer_12
   WavePlayer_13
   WavePlayer_14
   WavePlayer_15
   WavePlayer_16
   WavePlayer_17
   WavePlayer_18
   WavePlayer_19
   WavePlayer_20
  ]
  properties {weights [1,1,1,1,1,1,1,1,1,1,1,1]}
 }
 {id=WavePlayer_10, properties {soundWaveAssetPtr=&/FS_Henchman_Stone_Walk_08}}
 {id=WavePlayer_11, properties {soundWaveAssetPtr=&/FS_Henchman_Stone_Walk_07}}
 {id=WavePlayer_12, properties {soundWaveAssetPtr=&/FS_Henchman_Stone_Walk_06}}
 {id=WavePlayer_13, properties {soundWaveAssetPtr=&/FS_Henchman_Stone_Walk_05}}
 {id=WavePlayer_14, properties {soundWaveAssetPtr=&/FS_Henchman_Stone_Walk_04}}
 {id=WavePlayer_15, properties {soundWaveAssetPtr=&/FS_Henchman_Stone_Walk_03}}
 {id=WavePlayer_16, properties {soundWaveAssetPtr=&/FS_Henchman_Stone_Walk_02}}
 {id=WavePlayer_17, properties {soundWaveAssetPtr=&/FS_Henchman_Stone_Walk_01}}
 {id=WavePlayer_18, properties {soundWaveAssetPtr=&/FS_Henchman_Stone_Walk_12}}
 {id=WavePlayer_19, properties {soundWaveAssetPtr=&/FS_Henchman_Stone_Walk_11}}
 {id=WavePlayer_20, properties {soundWaveAssetPtr=&/FS_Henchman_Stone_Walk_10}}
 {id=WavePlayer_5, properties {soundWaveAssetPtr=&FS_Henchman_Stone_Walk_04}}
 {id=WavePlayer_6, properties {soundWaveAssetPtr=&FS_Henchman_Stone_Walk_03}}
 {id=WavePlayer_7, properties {soundWaveAssetPtr=&FS_Henchman_Stone_Walk_02}}
 {id=WavePlayer_8, properties {soundWaveAssetPtr=&FS_Henchman_Stone_Walk_01}}
 {id=WavePlayer_9, properties {soundWaveAssetPtr=&/FS_Henchman_Stone_Walk_09}}
]
properties {
 attenuationSettings=/Game/Sounds/Attenuation/Fort_Player_Footsteps_3P_Sprint_Athena_Att
 bPrimeOnLoad=true
 maxDistance=3584
 soundClassObject=&Fort_Player_Foley_3P_Enemy_Athena
 volumeMultiplier=0.5
 concurrencyOverrides {bLimitToOwner=true, maxCount=2, resolutionRule=StopOldest}
 concurrencySet [&Athena_Footstep_Local_Conc, &Athena_Footstep_Global_Conc]
}
